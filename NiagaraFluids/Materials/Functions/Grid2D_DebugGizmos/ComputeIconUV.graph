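# Material function in the Niagara Fluids Grid2D debug gizmo set that computes UVs and a mask for placing one icon in a tiled cell.
# It scales UV around the center, multiplies by Size and subtracts Offset, outputting the fractional part as the icon UV.
# A Custom HLSL node uses the rounded-up coordinate to mask all but one cell, with CropLeft trimming that cell's high-U edge despite its name.
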MaterialFunction /NiagaraFluids/Materials/Functions/Grid2D_DebugGizmos/ComputeIconUV
graph {
 properties {bExposeToLibrary=true}
 nodes [
  n1 = TextureCoordinate()
  n2 = FunctionInput(Preview=Constant2Vector() {g=1, r=1}) {bUsePreviewValueAsDefault=true, inputName=Scale, inputType=FunctionInput_Vector2, previewValue=<1 1 1 1>}
  n3 = Divide(A=Subtract(A=FunctionInput(Preview=@n1) {bUsePreviewValueAsDefault=true, inputName=UV, inputType=FunctionInput_Vector2}) {constB=0.5}, B=@n2)
  n4 = FunctionInput(Preview=Constant() {r=8}) {bUsePreviewValueAsDefault=true, inputName=Size, inputType=FunctionInput_Scalar, previewValue=<8 8 8 8>}
  n5 = Constant2Vector()
  n6 = Subtract(
   A=Multiply(A=Add(A=@n3) {constB=0.5}, B=@n4) {constB=8}
   B=ComponentMask(None=FunctionInput(Preview=@n5) {bUsePreviewValueAsDefault=true, inputName=Offset, inputType=FunctionInput_Vector2}) {g=true, r=true}
  )
  MaterialExpressionFrac_5 = Frac(None=@n6)
  FunctionOutput(None=@MaterialExpressionFrac_5) {outputName=UV}
  MaterialExpressionCeil_0 = Ceil(None=@n6)
  MaterialExpressionFunctionInput_5 = FunctionInput(Preview=Constant()) {bUsePreviewValueAsDefault=true, inputName=CropLeft, inputType=FunctionInput_Scalar, previewValue=<8 8 8 8>}
  n7 = Custom!(CeilUV=@MaterialExpressionCeil_0, UV=@MaterialExpressionFrac_5, CropLeft=@MaterialExpressionFunctionInput_5) {
   code="float Mask = 0.0;\r\nif (CeilUV.x == 1.0 &&&& CeilUV.y == 1.0){\r\n\tMask = 1.0;\r\n}\r\nif (UV.x >= 1.0-CropLeft){\r\n\tMask = 0.0;\r\n}\r\nreturn Mask.x;"
   description=CeilUV
   inputs [
    {inputName=CeilUV, input {expression=&self:MaterialExpressionCeil_0}}
    {inputName=UV, input {expression=&self:MaterialExpressionFrac_5}}
    {inputName=CropLeft, input {expression=&self:MaterialExpressionFunctionInput_5}}
   ]
  }
  FunctionOutput(None=ComponentMask(None=@n7) {r=true}) {outputName=Mask}
 ]
}
